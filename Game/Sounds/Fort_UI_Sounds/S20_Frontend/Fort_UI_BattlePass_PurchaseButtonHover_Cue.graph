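# UI sound cue for hovering the purchase button on the Battle Pass frontend screen.
# A Mixer layers a generic frontend button click with a louder, dedicated purchase-button hover sound.
# Only the hover layer passes through a Modulator, which boosts and slightly randomizes its volume.
# The cue uses the UI sound class and overrides concurrency to a small per-owner voice limit that stops the oldest instance.
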SoundCue /Game/Sounds/Fort_UI_Sounds/S20_Frontend/Fort_UI_BattlePass_PurchaseButtonHover_Cue
root=Mixer_0
nodes [
 {id=Mixer_0, children [WavePlayer_1,Modulator_0], properties {inputVolume [1,1.75]}}
 {id=Modulator_0, children [WavePlayer_3], properties {volumeMax=1.3, volumeMin=1.2}}
 {id=WavePlayer_1, properties {soundWaveAssetPtr=/Game/Sounds/Fort_UI_Sounds/FrontEnd/New_FrontEnd/UI_FN_Button_Click_Main_01}}
 {id=WavePlayer_3, properties {soundWaveAssetPtr=&/UI_FrontEnd_BattlePass_PurchaseButton_Hover}}
]
properties {
 bOverrideConcurrency=true
 maxDistance=2097152
 soundClassObject=&Fort_UI
 volumeMultiplier=1
 concurrencyOverrides {bLimitToOwner=true, maxCount=2, resolutionRule=StopOldest, voiceStealReleaseTime=0.2, volumeScaleAttackTime=0.2}
}
