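# Sound cue for the UI sound played when the Battle Pass purchase button is unhovered in the Season 20 frontend.
# A single wave player runs through a modulator with fixed pitch and volume, so there is no variation, while cue multipliers lower both.
# It uses the Fort_UI sound class and overrides concurrency, allowing a few instances per owner and stopping the oldest.
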SoundCue /Game/Sounds/Fort_UI_Sounds/S20_Frontend/Fort_UI_BattlePass_PurchaseButtonUnhover_Cue
root=Modulator_0
nodes [
 {id=Modulator_0, children [WavePlayer_4], properties {pitchMax=1, pitchMin=1, volumeMax=1, volumeMin=1}}
 {id=WavePlayer_4, properties {soundWaveAssetPtr=&/UI_FrontEnd_BattlePass_PurchaseButton_UnHover}}
]
properties {
 bOverrideConcurrency=true
 maxDistance=2097152
 pitchMultiplier=0.9
 soundClassObject=&Fort_UI
 volumeMultiplier=0.6
 concurrencyOverrides {bLimitToOwner=true, maxCount=2, resolutionRule=StopOldest, voiceStealReleaseTime=0.2, volumeScaleAttackTime=0.2}
}
